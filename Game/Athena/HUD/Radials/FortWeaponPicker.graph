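# Widget Blueprint for the Athena HUD weapon picker wheel, a FortWeaponWheel subclass for changing weapons through a radial menu.
# It lays out the inherited slotted radial picker, highlight and item name over blurred background and pie-material images.
# A bottom bar with confirm keybind and Cancel is hidden for mobile, touch and some handheld platforms.
# Activation plays a sped-up blur intro; choosing an item animates the highlight's opacity and material radii.
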WidgetBlueprint /Game/Athena/HUD/Radials/FortWeaponPicker
animations [
 Intro {
  binding_count=1
  length_seconds=2.7076
  object=&self:Intro
  sequence {bindings [BackgroundBlur_35 {
     id=#1
     possessed_class=BackgroundBlur
     tracks [{
       type=MovieSceneFloatTrack
       display_name=Blur Strength
       sections [{type=MovieSceneFloatSection, channels [None {type=MovieSceneScriptingFloatChannel, keys [[5.625,0],[31,1.4],[59.7773,2]]}]}]
      }]
    }]}
 }
 SelectionSet {
  binding_count=1
  length_seconds=0.5
  object=&self:SelectionSet
  sequence {bindings [Image_RadialHighlight {
     id=#2
     possessed_class=Image
     tracks [
      {
       type=MovieSceneWidgetMaterialTrack
       sections [{
         type=MovieSceneParameterSection
         channels [
          OuterRadius_0-1 {type=MovieSceneScriptingFloatChannel, keys [[0,0.97],[10,1]]}
          InnerRadius_0-1 {type=MovieSceneScriptingFloatChannel, keys [[0,0.5603],[10,0.6071]]}
         ]
        }]
      }
      {
       type=MovieSceneFloatTrack
       display_name=Render Opacity
       sections [{type=MovieSceneFloatSection, channels [None {type=MovieSceneScriptingFloatChannel, keys [[0,0],[1.557,1]]}]}]
      }
     ]
    }]}
 }
]
graphs [
 Cosmetic {nodes [
   PlayIntroAnimation() {is=title K2Node_CustomEvent}
   > PlayAnimation(InAnimation=Get Intro(), PlaybackSpeed=5)
   Event On Item Chosen() {is=title K2Node_Event}
   > PlayAnimation(InAnimation=Get SelectionSet() -> (SelectionSet:Widget Animation Object Reference))
  ]}
 EventGraph {nodes [
   Event On Activated()
   > PlayIntroAnimation() {is=title K2Node_CallFunction}
   Reroute Node(InputPin:Wildcard) -> (OutputPin:Wildcard)
  ]}
]
variables []
widgets {
 info {inheritedWidgetCount=5, parentClass=/Script/FortniteUI.FortWeaponWheel, rootWidgetClass=&Overlay, widgetCount=27}
 widgets [
  Overlay_1 {widgetClassPath=&Overlay, widget=Overlay_1}
  BackgroundBlur_35 {
   widgetClassPath=&BackgroundBlur
   parent=Overlay_1
   widget=BackgroundBlur_35
   properties {
    blurStrength=0.5
    horizontalAlignment=HAlign_Center
    verticalAlignment=VAlign_Center
    lowQualityFallbackBrush {
     bIsDynamicallyLoaded=false
     drawAs=Image
     imageType=NoImage
     mirroring=NoMirror
     resourceName=None
     resourceObject=null
     tiling=NoTile
     imageSize=<0 0>
     margin=<ltrb 0 0 0 0>
     outlineSettings {
      bUseBrushTransparency=false
      roundingType=HalfHeightRadius
      width=0
      color {colorUseRule=UseColor_Specified, specifiedColor=<rgba 0 0 0 0>}
      cornerRadii=<0 0 0 0>
     }
     tintColor {colorUseRule=UseColor_Specified, specifiedColor=<rgba 0.0052 0.0052 0.0056 0.4>}
     uVRegion {bIsValid=false, max=<0 0>, min=<0 0>}
    }
    padding=<ltrb -500 -500 -500 -500>
   }
   slot {class=OverlaySlot, properties {horizontalAlignment=HAlign_Fill, verticalAlignment=VAlign_Fill}}
  }
  HorizontalBox_0 {
   widgetClassPath=&HorizontalBox
   parent=Overlay_1
   widget=HorizontalBox_0
   slot {class=OverlaySlot, properties {horizontalAlignment=HAlign_Center, verticalAlignment=VAlign_Center}}
  }
  VerticalBox_1 {widgetClassPath=&VerticalBox, parent=HorizontalBox_0, widget=VerticalBox_1, slot {class=HorizontalBoxSlot}}
  SizeBox_7 {
   widgetClassPath=&SizeBox
   parent=VerticalBox_1
   widget=SizeBox_7
   properties {bOverride_HeightOverride=true, bOverride_WidthOverride=true, heightOverride=700, widthOverride=700}
   slot {class=VerticalBoxSlot}
  }
  Overlay_0 {widgetClassPath=&Overlay, parent=SizeBox_7, widget=Overlay_0, slot {class=SizeBoxSlot}}
  Image_Radial_BG {
   widgetClassPath=&Image
   parent=Overlay_0
   widget=Image_Radial_BG
   properties {
    visibility=HitTestInvisible
    brush {
     bIsDynamicallyLoaded=false
     drawAs=Image
     imageType=NoImage
     mirroring=NoMirror
     resourceName=None
     resourceObject=/Game/UI/Items/Materials/MI-AthenaPieShadow
     tiling=NoTile
     imageSize=<128 128>
     margin=<ltrb 0 0 0 0>
     outlineSettings {
      bUseBrushTransparency=false
      roundingType=HalfHeightRadius
      width=0
      color {colorUseRule=UseColor_Specified, specifiedColor=<rgba 0 0 0 0>}
      cornerRadii=<0 0 0 0>
     }
     tintColor {colorUseRule=UseColor_Specified, specifiedColor=<rgba 1 1 1 1>}
     uVRegion {bIsValid=false, max=<0 0>, min=<0 0>}
    }
    colorAndOpacity=<rgba 0.0052 0.0052 0.0056 0.4>
    renderTransform {angle=0, scale=<1 1>, shear=<0 0>, translation=<20 20>}
   }
   slot {class=OverlaySlot, properties {horizontalAlignment=HAlign_Fill, verticalAlignment=VAlign_Fill}}
  }
  Image_Radial {
   widgetClassPath=&Image
   parent=Overlay_0
   widget=Image_Radial
   properties {
    visibility=HitTestInvisible
    brush {
     bIsDynamicallyLoaded=false
     drawAs=Image
     imageType=NoImage
     mirroring=NoMirror
     resourceName=None
     resourceObject=/Game/UI/Items/Materials/MI-AthenaPieMenu
     tiling=NoTile
     imageSize=<128 128>
     margin=<ltrb 0 0 0 0>
     outlineSettings {
      bUseBrushTransparency=false
      roundingType=HalfHeightRadius
      width=0
      color {colorUseRule=UseColor_Specified, specifiedColor=<rgba 0 0 0 0>}
      cornerRadii=<0 0 0 0>
     }
     tintColor {colorUseRule=UseColor_Specified, specifiedColor=<rgba 1 1 1 1>}
     uVRegion {bIsValid=false, max=<0 0>, min=<0 0>}
    }
    colorAndOpacity=<rgba 0.0273 0.0273 0.0273 1>
   }
   slot {class=OverlaySlot, properties {horizontalAlignment=HAlign_Fill, verticalAlignment=VAlign_Fill}}
  }
  Image_RadialHighlight {
   widgetClassPath=&Image
   bInherited=true
   bIsVariable=true
   parent=Overlay_0
   widget=Image_RadialHighlight
   properties {
    visibility=HitTestInvisible
    brush {
     bIsDynamicallyLoaded=false
     drawAs=Image
     imageType=NoImage
     mirroring=NoMirror
     resourceName=None
     resourceObject=/Game/UI/Items/Materials/MI-AthenaPieMenu_Selection
     tiling=NoTile
     imageSize=<32 32>
     margin=<ltrb 0 0 0 0>
     outlineSettings {
      bUseBrushTransparency=false
      roundingType=HalfHeightRadius
      width=0
      color {colorUseRule=UseColor_Specified, specifiedColor=<rgba 0 0 0 0>}
      cornerRadii=<0 0 0 0>
     }
     tintColor {colorUseRule=UseColor_Specified, specifiedColor=<rgba 1 1 1 1>}
     uVRegion {bIsValid=false, max=<0 0>, min=<0 0>}
    }
   }
   slot {class=OverlaySlot, properties {horizontalAlignment=HAlign_Fill, verticalAlignment=VAlign_Fill}}
  }
  RadialMenu_Picker {
   widgetClassPath=/Game/Athena/HUD/ContextMenus/FortSlottedRadialMenu
   bInherited=true
   bIsVariable=true
   parent=Overlay_0
   widget=RadialMenu_Picker
   properties {
    entryWidgetClass=/Game/UI/InGame/HUD/AthenaRadial/FortSelectableRadialEntry.FortSelectableRadialEntry_C
    joystickDeadZone=0.4
    mouseDeadZoneRadiusPercentage=0.3
    navigation=&self:WidgetTree.RadialMenu_Picker.WidgetNavigation_0
    orderingMode=Clockwise
    visibility=Visible
    radialBoxSettings {angleBetweenItems=0, bClockwiseOrder=false, bDistributeItemsEvenly=true, marginSize=0, sectorCentralAngle=360, startingAngle=180}
   }
   slot {class=OverlaySlot, properties {horizontalAlignment=HAlign_Fill, verticalAlignment=VAlign_Fill}}
  }
  HorizontalBox_3 {
   widgetClassPath=&HorizontalBox
   parent=Overlay_1
   widget=HorizontalBox_3
   properties {renderTransform {angle=0, scale=<1 1>, shear=<0 0>, translation=<0 -87>}}
   slot {class=OverlaySlot, properties {horizontalAlignment=HAlign_Center, verticalAlignment=VAlign_Bottom}}
  }
  OkVertBox {widgetClassPath=&HorizontalBox, parent=HorizontalBox_3, widget=OkVertBox, slot {class=HorizontalBoxSlot}}
  Text_ItemName {
   widgetClassPath=&CommonTextBlock
   bInherited=true
   bIsVariable=true
   parent=OkVertBox
   widget=Text_ItemName
   properties {
    justification=Center
    text="{Item Name}"
    textTransformPolicy=ToUpper
    font {
     bForceMonospaced=false
     bMaterialIsStencil=false
     fontMaterial=null
     fontObject=&Burbank
     letterSpacing=0
     monospacedWidth=1
     size=43
     skewAmount=0
     typefaceFontName=Burbank-PlayerPower
     outlineSettings {
      bApplyOutlineToDropShadows=false
      bMiteredCorners=false
      bSeparateFillAlpha=false
      outlineMaterial=&M-FontShadow
      outlineSize=3
      outlineColor=<rgba 0 0 0 0.5>
     }
    }
    shadowColorAndOpacity=<rgba 0 0 0 0.5>
    shadowOffset=<5 5>
   }
   slot {class=HorizontalBoxSlot}
  }
  CommonVisibilityWidget_2 {
   widgetClassPath=&CommonVisibilityWidgetLegacy
   parent=Overlay_1
   widget=CommonVisibilityWidget_2
   properties {bShowForAndroid=false, bShowForErebus=false, bShowForIOS=false, bShowForSwitch2=false, bShowForTouch=false}
   slot {class=OverlaySlot, properties {horizontalAlignment=HAlign_Fill, verticalAlignment=VAlign_Bottom}}
  }
  BottomBarOverlay {widgetClassPath=&Overlay, bIsVariable=true, parent=CommonVisibilityWidget_2, widget=BottomBarOverlay, slot {class=BorderSlot}}
  CommonBorder_1 {
   widgetClassPath=&CommonBorder
   parent=BottomBarOverlay
   widget=CommonBorder_1
   properties {padding=<ltrb 0 0 0 0>}
   slot {class=OverlaySlot, properties {horizontalAlignment=HAlign_Fill, verticalAlignment=VAlign_Fill}}
  }
  Overlay {widgetClassPath=&Overlay, parent=CommonBorder_1, widget=Overlay, slot {class=BorderSlot, properties {padding=<ltrb 0 0 0 0>}}}
  Image_bottombar {
   widgetClassPath=&Image
   bIsVariable=true
   parent=Overlay
   widget=Image_bottombar
   properties {
    brush {
     bIsDynamicallyLoaded=false
     drawAs=Image
     imageType=NoImage
     mirroring=NoMirror
     resourceName=None
     resourceObject=null
     tiling=NoTile
     imageSize=<32 45>
     margin=<ltrb 0 0 0 0>
     outlineSettings {
      bUseBrushTransparency=false
      roundingType=HalfHeightRadius
      width=0
      color {colorUseRule=UseColor_Specified, specifiedColor=<rgba 0 0 0 0>}
      cornerRadii=<0 0 0 0>
     }
     tintColor {colorUseRule=UseColor_Specified, specifiedColor=<rgba 1 1 1 1>}
     uVRegion {bIsValid=false, max=<0 0>, min=<0 0>}
    }
    colorAndOpacity=<rgba 0.006 0.011 0.0242 0.8>
   }
   slot {class=OverlaySlot, properties {horizontalAlignment=HAlign_Fill, verticalAlignment=VAlign_Fill}}
  }
  VerticalBox {
   widgetClassPath=&VerticalBox
   parent=Overlay
   widget=VerticalBox
   slot {class=OverlaySlot, properties {horizontalAlignment=HAlign_Fill, verticalAlignment=VAlign_Fill}}
  }
  SafeZone {
   widgetClassPath=&SafeZone
   bIsVariable=true
   parent=VerticalBox
   widget=SafeZone
   properties {padBottom=false, padLeft=false, padRight=false}
   slot {class=VerticalBoxSlot}
  }
  Spacer {widgetClassPath=&Spacer, parent=VerticalBox, widget=Spacer, properties {size=<1 45>}, slot {class=VerticalBoxSlot}}
  HorizontalBox_4 {
   widgetClassPath=&HorizontalBox
   parent=BottomBarOverlay
   widget=HorizontalBox_4
   slot {class=OverlaySlot, properties {horizontalAlignment=HAlign_Center, verticalAlignment=VAlign_Bottom}}
  }
  EmoteHotkey {
   widgetClassPath=&HorizontalBox
   parent=HorizontalBox_4
   widget=EmoteHotkey
   slot {class=HorizontalBoxSlot, properties {padding=<ltrb 0 0 10 0>}}
  }
  SizeBox_1 {
   widgetClassPath=&SizeBox
   parent=EmoteHotkey
   widget=SizeBox_1
   properties {bOverride_HeightOverride=true, heightOverride=40, widthOverride=30}
   slot {class=HorizontalBoxSlot}
  }
  Keybind_Confirm {
   widgetClassPath=/FortUILibrary/Input/KeybindWidget
   bInherited=true
   bIsVariable=true
   parent=SizeBox_1
   widget=Keybind_Confirm
   properties {forcedHoldKeybindStatus=NeverShowHold}
   slot {class=SizeBoxSlot, properties {padding=<ltrb 0 0 5 0>}}
  }
  CommonTextBlock {
   widgetClassPath=&CommonTextBlock
   parent=EmoteHotkey
   widget=CommonTextBlock
   properties {
    justification=Center
    text=Change Weapon
    font {
     bForceMonospaced=false
     bMaterialIsStencil=false
     fontMaterial=null
     fontObject=&BurbankSmall
     letterSpacing=0
     monospacedWidth=1
     size=18
     skewAmount=0
     typefaceFontName=Bold
     outlineSettings {
      bApplyOutlineToDropShadows=false
      bMiteredCorners=false
      bSeparateFillAlpha=false
      outlineMaterial=null
      outlineSize=0
      outlineColor=<rgba 0 0 0 1>
     }
    }
    shadowOffset=<0 0>
   }
   slot {class=HorizontalBoxSlot, properties {horizontalAlignment=HAlign_Center, verticalAlignment=VAlign_Center}}
  }
  Button_Cancel {
   widgetClassPath=/FortUILibrary/Input/InputReflectorButton
   bInherited=true
   bIsVariable=true
   parent=HorizontalBox_4
   widget=Button_Cancel
   properties {
    buttonText=Cancel
    controllerInputStyle=/FortUILibrary/Foundation/Buttons/ButtonStyles/ButtonStyle-Empty.ButtonStyle-Empty_C
    onButtonBaseClicked=(&self:WidgetTree.Button_Cancel.HandleOnButtonClicked)
    onButtonBaseDoubleClicked=(&self:WidgetTree.Button_Cancel.HandleOnButtonDoubleClicked)
    onButtonBaseHovered=(&self:WidgetTree.Button_Cancel.HandleOnButtonHovered)
    onButtonBaseUnhovered=(&self:WidgetTree.Button_Cancel.HandleOnButtonUnhovered)
    onSelectedChangedBase=(&self:WidgetTree.Button_Cancel.HandleOnSelectedChanged)
    style=/FortUILibrary/Foundation/Buttons/ButtonStyles/TopBar/ButtonStyle-OutlineEmptyFill.ButtonStyle-OutlineEmptyFill_C
    visibility=Visible
    pressedSlateSoundOverride {resourceObject=/Game/Sounds/Fort_UI_Sounds/FrontEnd/fort_ui_dismiss_03_Cue}
    triggeringInputAction {dataTable=/Game/UI/InGame/HUD/AthenaRadial/EmoteWheelInternalBindings, rowName=ClosePicker}
   }
   slot {class=HorizontalBoxSlot, properties {verticalAlignment=VAlign_Center}}
  }
 ]
}
